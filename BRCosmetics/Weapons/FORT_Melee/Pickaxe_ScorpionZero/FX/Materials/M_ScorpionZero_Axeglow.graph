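# Translucent unlit material for Niagara mesh particles that provides the axe glow effect on the Scorpion Zero pickaxe.
# UV1's V axis blends red-orange into green-teal, multiplied by particle color and fed through MF_FX_EmissiveAdjustments.
# A separate UV1 gradient, skewed by U, slides with particle alpha.
# A vertex offset term combines a push toward the camera with an adjustable push along normals.
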Material /BRCosmetics/Weapons/FORT_Melee/Pickaxe_ScorpionZero/FX/Materials/M_ScorpionZero_Axeglow
graph {
 properties {
  bUsedWithNiagaraMeshParticles=true
  blendMode=BLEND_Translucent
  pixelDepthOffsetMode=PDOM_Legacy
  shadingModel=MSM_Unlit
  usedShadingModels=Unlit
 }
 nodes [
  n1 = Subtract(A=CameraPositionWS(), B=WorldPosition().XYZ)
  Add(
   A=Multiply(A=Normalize(VectorInput=@n1), B=ScalarParameter() {defaultValue=1, group=Vertex Offset, parameterName=Camera Offset, sliderMax=1})
   B=Multiply(A=VertexNormalWS(), B=ScalarParameter() {group=Vertex Offset, parameterName=Vertex Normal Push})
  )
  n2 = TextureCoordinate() {coordinateIndex=1}
  n3 = Multiply(A=ComponentMask(None=TextureCoordinate() {coordinateIndex=1}) {r=true}) {constB=0.2}
  n4 = ParticleColor()
  ComponentMask(None=Add(A=Add(A=@n2, B=@n3), B=LinearInterpolate(Alpha=@n4.A) {constA=-1})) {g=true}
  n5 = ComponentMask(None=TextureCoordinate() {coordinateIndex=1}) {g=true}
  n6 = LinearInterpolate(
   A=Constant3Vector() {constant=<rgba 1 0.0275 0 1>}
   B=Constant3Vector() {constant=<rgba 1 0.376 0 1>}
   Alpha=SmoothStep(Value=@n5) {constMax=0.5}
  )
  n7 = LinearInterpolate(
   A=Constant3Vector() {constant=<rgba 0.1978 0.7083 0.1367 1>}
   B=Constant3Vector() {constant=<rgba 0.147 1 0.5957 1>}
   Alpha=SmoothStep(Value=@n5) {constMin=0.5}
  )
  n8 = SmoothStep(Value=@n5) {constMax=0.75, constMin=0.25}
  MaterialFunctionCall(Emissive=Multiply(A=LinearInterpolate(A=@n6, B=@n7, Alpha=@n8), B=@n4.RGB)) {materialFunction=&MF_FX_EmissiveAdjustments}
 ]
}
