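# Translucent unlit Niagara sprite material that draws the hand icon for the Scorpion Zero pickaxe effects.
# A smoothstep-sharpened hand mask gets a vertical orange to red-orange gradient, with a green outline from a second mask lerped on top.
# Particle color tints the result, then a dynamic parameter drives a hue shift before emissive adjustments.
# Opacity comes from particle alpha multiplied by the sharpened hand mask.
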Material /BRCosmetics/Weapons/FORT_Melee/Pickaxe_ScorpionZero/FX/Materials/M_ScorpionZero_HandIcon
graph {
 properties {bUsedWithNiagaraSprites=true, blendMode=BLEND_Translucent, pixelDepthOffsetMode=PDOM_Legacy, shadingModel=MSM_Unlit, usedShadingModels=Unlit}
 nodes [
  n1 = TextureSample() {samplerType=SAMPLERTYPE_LinearGrayscale, texture=/BRCosmetics/Weapons/FORT_Melee/Pickaxe_ScorpionZero/FX/Textures/T_ScorpionZeroHand_Mask_M}
  n2 = SmoothStep(Value=@n1.R) {constMin=0.9}
  n3 = Constant3Vector() {constant=<rgba 1 0.5856 0 1>}
  n4 = Constant3Vector() {constant=<rgba 1 0.0843 0 1>}
  n5 = SmoothStep(Value=ComponentMask(None=TextureCoordinate()) {g=true}) {constMax=0.65, constMin=0.25}
  n6 = TextureSample() {
   samplerType=SAMPLERTYPE_LinearGrayscale
   texture=/BRCosmetics/Weapons/FORT_Melee/Pickaxe_ScorpionZero/FX/Textures/T_ScorpionZeroHand_Outline_M
  }
  n7 = SmoothStep(Value=@n6.R) {constMin=0.9}
  n8 = LinearInterpolate(
   A=Multiply(A=@n2, B=LinearInterpolate(A=@n3, B=@n4, Alpha=@n5))
   B=Multiply(A=@n7, B=Constant3Vector() {constant=<rgba 0.0983 0.599 0.0254 1>})
   Alpha=@n7
  )
  n9 = ParticleColor()
  MaterialFunctionCall(Emissive=MaterialFunctionCall(Hue Shift Percentage=DynamicParameter() {paramNames [Hueshift,Param2,Param3,Param4]}.Hueshift, Texture=Multiply(A=@n8, B=@n9.RGB)) {materialFunction=&HueShift}.Result) {materialFunction=&MF_FX_EmissiveAdjustments}
  Multiply(A=@n9.A, B=@n2)
 ]
}
